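# Physics asset for the CyborgWarrior medium male backpack cosmetic, setting up physics for a cover and two pages on the pack.
# Capsule bodies cover the main body, a tiny base pivot for each flap, and a larger dynamic piece on each.
# Each base joins the body with free swing and twist.
# Each dynamic piece hangs off its base with narrow, per-piece swing limits and a zero twist limit.
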PhysicsAsset /BRCosmetics/Accessories/FORT_Backpacks/Backpack_M_MED_CyborgWarrior/Meshes/M_MED_CyborgWarrior_Pack_Physics
bodies {
 base_cover [{shapeType=Capsule, length=0.505, radius=0.505, shapeName=base_cover_capsule, rotation=<rot 0 -90 0>}]
 base_page_01_cr [{shapeType=Capsule, length=0.505, radius=0.505, shapeName=base_page_01_cr_capsule, rotation=<rot 0 -90 0>}]
 base_page_02_cr [{shapeType=Capsule, length=0.505, radius=0.505, shapeName=base_page_02_cr_capsule, rotation=<rot 0 -90 0>}]
 body_cr [{
   shapeType=Capsule
   length=23.4775
   radius=6.5225
   shapeName=body_cr_capsule
   center=<0 4.1629 -5.7893>
   rotation=<rot 90 -179.6835 -179.6838>
  }]
 dyn_cover [{shapeType=Capsule, length=14.3727, radius=11.3106, center=<15.5072 0.0021 -10.0069>, rotation=<rot 0.1566 0.0449 -90>}]
 dyn_page_01_cr [{shapeType=Capsule, length=20.175, radius=1.4945, shapeName=dyn_page_01_cr_capsule, center=<17.7732 0 0>, rotation=<rot 90 0 0>}]
 dyn_page_02_cr [{shapeType=Capsule, length=10.4699, radius=1.9213, shapeName=dyn_page_02_cr_capsule, center=<18.1567 0 1.1653>, rotation=<rot 90 180 -90>}]
}
constraints [
 {
  bone1Name=base_page_01_cr
  bone2Name=body_cr
  swing1LimitDegrees=45
  swing1Motion=Free
  swing2LimitDegrees=45
  swing2Motion=Free
  twistLimitDegrees=45
  twistMotion=Free
 }
 {bone1Name=dyn_page_01_cr, bone2Name=base_page_01_cr, swing1LimitDegrees=8, swing2LimitDegrees=30, twistLimitDegrees=0}
 {
  bone1Name=base_page_02_cr
  bone2Name=body_cr
  swing1LimitDegrees=45
  swing1Motion=Free
  swing2LimitDegrees=45
  swing2Motion=Free
  twistLimitDegrees=45
  twistMotion=Free
 }
 {bone1Name=dyn_page_02_cr, bone2Name=base_page_02_cr, swing1LimitDegrees=8, swing2LimitDegrees=5, twistLimitDegrees=0}
 {bone1Name=dyn_cover, bone2Name=base_cover, swing1LimitDegrees=8, swing2LimitDegrees=20, twistLimitDegrees=0}
 {
  bone1Name=base_cover
  bone2Name=body_cr
  swing1LimitDegrees=45
  swing1Motion=Free
  swing2LimitDegrees=45
  swing2Motion=Free
  twistLimitDegrees=45
  twistMotion=Free
 }
]
